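# Niagara effect for the LettuceCat backpack that throws out colored VBucks mesh particles.
# User.Stage selects spawn rate and burst count, with rate also scaled by a system curve.
# User.bJumpReactivity reverses burst launch velocity and shortens lifetime, while front-end modifiers tone down count, scale and speed.
# Forces and drag move particles, sphere blocking volumes surround the emitter and camera, and a debug sprite emitter is disabled.
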NiagaraSystem /BRCosmetics/Accessories/FORT_Backpacks/Backpack_M_MED_LettuceCat/FX/NS_Backpack_LettuceCat
dependencies {
 usedDataInterfaces [
  &NiagaraDataInterfaceArrayFloat
  &NiagaraDataInterfaceArrayInt32
  &NiagaraDataInterfaceArrayColor
  &NiagaraDataInterfaceMeshRendererInfo
  &NiagaraDataInterfaceCurve
  &NiagaraDataInterfaceCamera
 ]
 usedDynamicInputs [
  &ClampFloat
  &VectorLength
  &Multiply_Vector
 ]
 usedModules [
  &SystemState
  &self:SystemSpawnScript.SystemScriptSource.SystemScriptGraph.NiagaraNodeAssignment_0.SetVariables_#1
  &self:SystemSpawnScript.SystemScriptSource.SystemScriptGraph.NiagaraNodeAssignment_1.SetVariables_#2
 ]
}
emitters {
 FN_Debug {
  bEnabled=false
  emitterUpdateScript [
   &EmitterState()
   &SpawnBurst_Instantaneous(Spawn Count=1, Use Spawn Probability=false)
  ]
  particleSpawnScript [/Niagara/Modules/Spawn/Initialization/InitializeParticle(
    Lifetime=5
    Module.Write Color=true
    Module.Write Lifetime=true
    Module.Write Mass=true
    Module.Write Material Random=true
    Module.Write Position=false
    Module.Write Scale=false
    Module.Write SpriteRotation=false
    Module.Write SpriteSize=true
    Position=&SimulationPosition()
    Sprite Size=<100 100>
   )]
  particleUpdateScript [
   &ParticleState()
   &FN_Timer()
   /SetVariables(Particles.Debug:NiagaraFloat=&MakeCustomFloatFromBool(Bool=$User.bJumpReactivity, True Float=1, False Float=0)) {isSetParameters=true}
   /SetVariables(Particles.Debug:NiagaraFloat=$System.PlayerVelocityNormalized) {enabled=false, isSetParameters=true}
   &DynamicMaterialParameters(
    Index 0 Param 1=$Particles.Debug
    Module.Param0WriteEnabled=true
    Module.Param1WriteEnabled=true
    Module.Param2WriteEnabled=true
    Module.Param3WriteEnabled=true
   )
   &CameraOffset(Camera Offset Amount=111)
  ]
  rendererClasses [&NiagaraSpriteRendererProperties]
  renderers [{rendererClass=&NiagaraSpriteRendererProperties}]
 }
 VBucks {
  emitterSpawnScript [
   /Game/Effects/Niagara/MasterMaterials/ToonShader/Template/NM_ToonShaderParameters(HighLight_Softness=0, RimLight_Width=0)
   /Game/Effects/Niagara/MasterMaterials/ToonShader/Template/NM_OutlineParameters()
  ]
  emitterUpdateScript [
   &EmitterState(Life Cycle Mode=::NewEnumerator1, Loop Behavior=::NewEnumerator1, Loop Duration Mode=::NewEnumerator1, UseLoopDelay=false)
   &SpawnRate(
    Spawn Probability=0.5
    SpawnRate=&Multiply_Float(
     A=&SelectFloatFromArray(Direct Array Index=$User.Stage, Float Selection Array=DataInterface() {propertyValues {FloatData [0,7,26]}})
     B=$System.SpawnRateCurve
    )
    Use Spawn Probability=true
   )
   &SpawnBurst_Instantaneous(
    Spawn Count=&MakeIntFromFloat(Float=&Multiply_Float(
      A=&MakeFloatFromInt(Int=&SelectIntFromArray(Direct Array Index=$User.Stage, Int Selection Array=DataInterface() {propertyValues {IntData [0,11,33]}}))
      B=&FN_FrontEndFloatModifier(
       Gameplay=&RandomRangeFloat(Minimum=0.333, Maximum=1)
       Front End=&RandomRangeFloat(Minimum=0.166, Maximum=0.5)
      )
     ))
    Spawn Group=1
    Use Spawn Probability=false
    UseLoopCountLimit=false
   )
   /SetVariables(
    Emitter.WindOscillateTime:NiagaraFloat=&Divide_Float(A=$Engine.Time, B=44)
    Emitter.WindRotateTime:NiagaraFloat=&Divide_Float(A=$Engine.Time, B=66)
   ) {isSetParameters=true}
   /SetVariables(
    Emitter.WindIntensity:NiagaraFloat=&Multiply_Float(A=&Sine(Normalized Angle=$Emitter.WindOscillateTime, Period=1, Scale=1, Bias=0), B=0)
    Emitter.WindVector:Vector3f=&MakeVector(
     X=&Add_Float(
      A=&Sine(Normalized Angle=$Emitter.WindRotateTime, Period=1, Scale=1, Bias=0)
      B=&Cosine(Normalized Angle=$Emitter.WindRotateTime, Period=1, Scale=1, Bias=0)
     )
     Y=&Subtract_Float(
      A=&Cosine(Normalized Angle=$Emitter.WindRotateTime, Period=1, Scale=1, Bias=0)
      B=&Sine(Normalized Angle=$Emitter.WindRotateTime, Period=1, Scale=1, Bias=0)
     )
     Z=0
    )
   ) {isSetParameters=true}
  ]
  particleSpawnScript [
   /Niagara/Modules/Spawn/Initialization/V2/InitializeParticle(
    Color=&SelectLinearColorFromArray(
     Array Sampling Mode=::NewEnumerator0
     Direct Array Index=?
     Color Selection Array=DataInterface() {propertyValues {ColorData [
        <rgba 0.2086 0.6172 0.3613 1>
        <rgba 0.8632 0.7682 0.0343 1>
        <rgba 0.9047 0.4287 0.1384 1>
        <rgba 0.8879 0.2705 0.6514 1>
        <rgba 0.956 0.2961 1 1>
        <rgba 0.4452 0.1248 0.6172 1>
       ]}}
     Random Seed=0
    )
    Lifetime=&MaskFloatBySpawnGroup(
     Masked Float=&RandomRangeFloat(Minimum=2.666, Maximum=4.333)
     Passthrough Float=&MakeCustomFloatFromBool(
      Bool=$User.bJumpReactivity
      True Float=&RandomRangeFloat(Minimum=1.666, Maximum=3.666)
      False Float=&RandomRangeFloat(Minimum=5.666, Maximum=8.333)
     )
    )
    Mesh Renderer Array Visibility Mode=::NewEnumerator3
    Mesh Renderer Info=DataInterface() {propertyValues {MeshRenderer=&self:VBucks_0.NiagaraMeshRendererProperties_1}}
    Mesh Scale Mode=::NewEnumerator3
    Mesh Uniform Scale=&FN_FrontEndFloatModifier(Gameplay=0.75, Front End=0.6)
    Position Offset=&RandomRangeVector(Minimum=<5 -13 0>, Maximum=<-7 -9 0>)
    Position Offset Coordinate Space=::Local
    UsePositionOffset=true
    Write Material Random=true
   )
   &InheritVelocity(Inherited Velocity Speed Limit=333, Speed Limit=true, UseSpeedThreshold=false)
   &AddVelocity(
    Velocity=&Add_Vector(
     A=<0 -33 11>
     B=&Multiply_VectorByFloat(
      Vector=<1 -33 166>
      Float=&Multiply_Float(
       A=&MakeCustomFloatFromBool(Bool=$User.bJumpReactivity, True Float=-1, False Float=1)
       B=&MakeFloatFromInt(Int=&SpawnGroup())
      )
     )
    )
    Velocity Speed Scale=&Multiply_Float(
     A=&FN_FrontEndFloatModifier(Gameplay=1, Front End=0.3)
     B=&RandomRangeFloat(Minimum=0.666, Maximum=1.333)
    )
   )
   &ApplyInitialForces()
   &InitialMeshOrientation(Use Rotation Vector=false)
  ]
  particleUpdateScript [
   &ParticleState()
   &ScaleMeshSize(Scale Factor=&VectorFromFloat(Value=&FloatFromCurve(
      FloatCurve=DataInterface() {propertyValues {Curve {keys [
          {arriveTangent=4.1785, interpMode=RCIM_Cubic, leaveTangent=8.7018, tangentMode=RCTM_Break, time=0, value=0}
          {arriveTangent=0.0224, interpMode=RCIM_Cubic, leaveTangent=0.0224, tangentMode=RCTM_Break, time=0.3, value=1}
         ]}}}
      CurveIndex=&MaskFloatBySpawnGroup(Masked Float=$Particles.Age, Passthrough Float=1)
      Scale Curve=&FloatFromCurve(FloatCurve=DataInterface() {propertyValues {Curve {keys [
           {interpMode=RCIM_Cubic, tangentMode=RCTM_Break, time=0.7, value=1}
           {arriveTangent=-9.1079, interpMode=RCIM_Cubic, tangentMode=RCTM_Break, time=1, value=0}
          ]}}})
     )))
   &GravityForce()
   &AccelerationForce(Acceleration=<0 0 333>)
   &PointForce(Force Falloff Distance=333, Force Strength=666, Module.Use Falloff=true, Module.UseLinearForce=false)
   &InheritSourceMovement(
    Applied Position Scale=&Multiply_VectorByFloat(
     Vector=<1 1 1>
     Float=&FloatFromCurve(FloatCurve=DataInterface() {propertyValues {Curve {keys [<at 0 1>,<at 0.6 0>]}}}, CurveIndex=$Particles.Age, Scale Curve=0.5)
    )
    ApplyForce=false
    ApplyPosition=true
    ApplyPositionFalloff=false
   )
   &UpdateMeshOrientation(Rotation Rate=&RandomRangeFloat(Minimum=-1, Maximum=1))
   &WindForce(Wind Speed=$Emitter.WindVector, Wind Speed Scale=$Emitter.WindIntensity)
   &AerodynamicDrag(Rotation Strength=2, Use Linear Drag=true, Use Rotational Drag=true)
   &SolveForcesAndVelocity(Clamp Velocity=false, Limit Acceleration=false)
   &AlignSpriteToMeshOrientation(SetAlignment=true, SetFacing=true)
   &self:SphereBlockingVolume(
    Pivot Offset:Vector3f=<0 0 -11>
    Sphere Origin:NiagaraPosition=$Engine.Emitter.SimulationPosition
    Sphere Radius:NiagaraFloat=&FN_FrontEndFloatModifier(Gameplay=30, Front End=20)
   ) {module=SphereBlockingVolume}
   &self:SphereBlockingVolume(Pivot Offset:Vector3f=<0 16 40>, Sphere Origin:NiagaraPosition=$Engine.Emitter.SimulationPosition, Sphere Radius=26) {module=SphereBlockingVolume001}
   &self:SphereBlockingVolume(
    Sphere Origin:Vector3f=&ReturnCameraPosition(Coordinate Space:ENiagaraCoordinateSpace=::Simulation, Camera Query:NiagaraDataInterfaceCamera=DataInterface())
    Sphere Radius=66
   ) {module=SphereBlockingVolume002}
  ]
  rendererClasses [&NiagaraMeshRendererProperties]
  renderers [{rendererClass=&NiagaraMeshRendererProperties, rendererIndex=0}]
 }
}
summary {
 systemName=NS_Backpack_LettuceCat
 emitters [
  {bEnabled=false, emitterName=FN_Debug, rendererClasses [&NiagaraSpriteRendererProperties]}
  {emitterName=VBucks, rendererClasses [&NiagaraMeshRendererProperties]}
 ]
 userVariables [
  User.bIsFrontEnd {type=NiagaraBool}
  User.bIsFrontEndPreview {type=NiagaraBool}
  User.bJumpReactivity {type=NiagaraBool}
  User.Stage {type=NiagaraInt32}
 ]
}
