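# Physics asset for the stage 01 LettuceCat back bling backpack, defining secondary motion for its top handle and two zipper pulls.
# A pack capsule and a zipper root box act as anchor bodies, with small shapes on each dyn_ bone.
# Both zipper pulls hinge from the zipper root with wide swing and very tight twist limits.
# The handle capsule is constrained to the pack with narrower swing limits.
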PhysicsAsset /BRCosmetics/Accessories/FORT_Backpacks/Backpack_M_MED_LettuceCat/Meshes/M_MED_LettuceCat_Pack_Physics_stage_01
bodies {
 dyn_top_handle [{shapeType=Capsule, length=3, radius=2, shapeName=dyn_top_handle_capsule, center=<2.1051 -0.3525 0>, rotation=<rot 0.0007 0 -90>}]
 dyn_zipper_01 [{shapeType=Box, extentX=2.3, extentY=0.6, extentZ=0.3, shapeName=dyn_zipper_01_box, center=<1.4388 0 0.0433>}]
 dyn_zipper_02 [{shapeType=Box, extentX=2.3, extentY=0.6, extentZ=0.3, shapeName=dyn_zipper_02_box, center=<1.474 0 0.0331>}]
 pack [{shapeType=Capsule, length=6.1422, radius=7.642, shapeName=pack_capsule, center=<0 -7.0151 -1.03>, rotation=<rot 90 -0.0099 179.9899>}]
 zipper_root [{shapeType=Box, extentX=5, extentY=4.5, extentZ=1.5, shapeName=zipper_root_box, center=<0.3033 0.5352 -0.7919>}]
}
constraints [
 {bone1Name=dyn_zipper_01, bone2Name=zipper_root, swing1LimitDegrees=60, swing2LimitDegrees=90, twistLimitDegrees=3}
 {bone1Name=dyn_zipper_02, bone2Name=zipper_root, swing1LimitDegrees=60, swing2LimitDegrees=90, twistLimitDegrees=3}
 {bone1Name=dyn_top_handle, bone2Name=pack, swing1LimitDegrees=25, swing2LimitDegrees=25, twistLimitDegrees=10}
]
