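# Physics asset for stage 02 of the LettuceCat cosmetic backpack, defining secondary motion for its money bills, zippers and top handle.
# Small boxes cover seven bills and two zipper pulls, a capsule covers the handle, and larger anchor shapes cover pack, money root and zipper root.
# Angular-limited constraints tie left and right bills to the money root, top bills and handle to the pack, and zippers to their root with wider swing.
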PhysicsAsset /BRCosmetics/Accessories/FORT_Backpacks/Backpack_M_MED_LettuceCat/Meshes/M_MED_LettuceCat_Pack_Physics_stage_02
bodies {
 dyn_money_01_left_bill_01 [{
   shapeType=Box
   extentX=4
   extentY=3.8
   extentZ=0.3
   shapeName=dyn_money_01_left_bill_01_box
   center=<3.1397 -0.1036 0.4697>
   rotation=<rot 14.9607 -5.1508 -0.9003>
  }]
 dyn_money_01_left_bill_02 [{
   shapeType=Box
   extentX=3
   extentY=4
   extentZ=0.3
   shapeName=dyn_money_01_left_bill_02_box
   center=<2.4552 0 -0.0076>
   rotation=<rot 10 0.0002 -0.0001>
  }]
 dyn_money_01_right_bill_01 [{
   shapeType=Box
   extentX=3.4293
   extentY=3.8
   extentZ=0.3
   shapeName=dyn_money_01_right_bill_01_box
   center=<3.4113 -0.0183 0.0985>
   rotation=<rot 14.9988 4.996 -4.9991>
  }]
 dyn_money_01_right_bill_02 [{
   shapeType=Box
   extentX=3.2987
   extentY=3.861
   extentZ=0.3
   shapeName=dyn_money_01_right_bill_02_box
   center=<2.0547 0 0.3196>
   rotation=<rot 5 -0.0003 0.0001>
  }]
 dyn_money_01_top_bill_01 [{
   shapeType=Box
   extentX=3.5
   extentY=4
   extentZ=0.3
   shapeName=dyn_money_01_top_bill_01_box
   center=<1.5587 0.0832 -0.0976>
   rotation=<rot 0.0003 -14.9999 -0.0004>
  }]
 dyn_money_01_top_bill_02 [{
   shapeType=Box
   extentX=3.5
   extentY=5
   extentZ=0.5
   shapeName=dyn_money_01_top_bill_02_box
   center=<1.9286 0.1832 -0.1605>
   rotation=<rot -5.4155 -5.0033 -4.5456>
  }]
 dyn_money_01_top_bill_03 [{
   shapeType=Box
   extentX=2.5
   extentY=3
   extentZ=0.3
   shapeName=dyn_money_01_top_bill_03_box
   center=<0.7846 -0.018 -0.0295>
   rotation=<rot -8.0661 3.3935 6.0226>
  }]
 dyn_top_handle [{shapeType=Capsule, length=3, radius=2, shapeName=dyn_top_handle_capsule, center=<2.1051 -0.3525 0>, rotation=<rot 0.0007 0 -90>}]
 dyn_zipper_01 [{shapeType=Box, extentX=2.3, extentY=0.6, extentZ=0.3, shapeName=dyn_zipper_01_box, center=<1.4246 0 0.0708>}]
 dyn_zipper_02 [{shapeType=Box, extentX=2.3, extentY=0.6, extentZ=0.3, shapeName=dyn_zipper_02_box, center=<1.4246 0 0.0708>}]
 money_01_root [{shapeType=Box, extentX=25.5, extentY=12, extentZ=11.5, shapeName=money_01_root_box, center=<0 -1.5979 -18.0697>}]
 pack [{shapeType=Capsule, length=6.1422, radius=7.642, shapeName=pack_capsule, center=<0 -7.0151 -1.03>, rotation=<rot 90 -0.0099 179.9899>}]
 zipper_root [{shapeType=Box, extentX=4.5, extentY=3.5, extentZ=1.5, shapeName=zipper_root_box, center=<0 0.5352 -0.6165>}]
}
constraints [
 {bone1Name=dyn_zipper_01, bone2Name=zipper_root, swing1LimitDegrees=60, swing2LimitDegrees=90, twistLimitDegrees=3}
 {bone1Name=dyn_zipper_02, bone2Name=zipper_root, swing1LimitDegrees=60, swing2LimitDegrees=90, twistLimitDegrees=3}
 {bone1Name=dyn_top_handle, bone2Name=pack, swing1LimitDegrees=25, swing2LimitDegrees=25, twistLimitDegrees=10}
 {bone1Name=dyn_money_01_left_bill_01, bone2Name=money_01_root, swing1LimitDegrees=20, swing2LimitDegrees=40, twistLimitDegrees=5}
 {bone1Name=dyn_money_01_left_bill_02, bone2Name=money_01_root, swing1LimitDegrees=20, swing2LimitDegrees=40, twistLimitDegrees=5}
 {bone1Name=dyn_money_01_right_bill_02, bone2Name=money_01_root, swing1LimitDegrees=20, swing2LimitDegrees=40, twistLimitDegrees=5}
 {bone1Name=dyn_money_01_right_bill_01, bone2Name=money_01_root, swing1LimitDegrees=20, swing2LimitDegrees=40, twistLimitDegrees=5}
 {bone1Name=dyn_money_01_top_bill_03, bone2Name=pack, swing1LimitDegrees=20, swing2LimitDegrees=40, twistLimitDegrees=5}
 {bone1Name=dyn_money_01_top_bill_02, bone2Name=pack, swing1LimitDegrees=20, swing2LimitDegrees=40, twistLimitDegrees=5}
 {bone1Name=dyn_money_01_top_bill_01, bone2Name=pack, swing1LimitDegrees=10, swing2LimitDegrees=40, twistLimitDegrees=5}
]
