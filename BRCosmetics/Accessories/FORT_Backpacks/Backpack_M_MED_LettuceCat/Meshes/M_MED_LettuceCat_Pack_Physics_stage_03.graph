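# Physics asset for stage 03 of the LettuceCat backpack cosmetic, defining collision bodies and joints for its dangling money bills, top handle and two zipper pulls.
# Thin boxes stand in for the bills and zippers, with capsules for the handle, pack and money root.
# Bills and handle constrain directly to the pack, while zipper pulls hang from zipper_root with wider swing.
# All joints keep twist narrow, and money_02_root has no constraint.
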PhysicsAsset /BRCosmetics/Accessories/FORT_Backpacks/Backpack_M_MED_LettuceCat/Meshes/M_MED_LettuceCat_Pack_Physics_stage_03
bodies {
 dyn_money_02_left_bill_01 [{
   shapeType=Box
   extentX=4
   extentY=6
   extentZ=0.3
   shapeName=dyn_money_02_left_bill_01_box
   center=<1.7583 -0.2924 -0.0697>
   rotation=<rot -10 0.0007 -0.0004>
  }]
 dyn_money_02_left_bill_02 [{
   shapeType=Box
   extentX=4.5
   extentY=3.5
   extentZ=0.3
   shapeName=dyn_money_02_left_bill_02_box
   center=<1.1347 -0.4985 0.1368>
   rotation=<rot 0.0001 -19.9988 -0.0004>
  }]
 dyn_money_02_right_bill_01 [{
   shapeType=Box
   extentX=5.5
   extentY=4
   extentZ=0.3
   shapeName=dyn_money_02_right_bill_01_box
   center=<2.4571 0.07 0.2169>
   rotation=<rot 4.9998 -5.0022 5.0015>
  }]
 dyn_money_02_right_bill_02 [{
   shapeType=Box
   extentX=5
   extentY=7
   extentZ=0.3
   shapeName=dyn_money_02_right_bill_02_box
   center=<2.3101 0 0.1355>
   rotation=<rot 5 -0.0034 5.0009>
  }]
 dyn_money_02_top_bill_01 [{
   shapeType=Box
   extentX=5.5
   extentY=7.5
   extentZ=0.5
   shapeName=dyn_money_02_top_bill_01_box
   center=<1.8807 0.8824 0.1999>
   rotation=<rot -4.9994 0 -0.0003>
  }]
 dyn_money_02_top_bill_02 [{shapeType=Box, extentX=4, extentY=8.5, extentZ=0.5, shapeName=dyn_money_02_top_bill_02_box, center=<1.8043 0.4444 0.1906>}]
 dyn_money_02_top_bill_03 [{
   shapeType=Box
   extentX=4.5
   extentY=10
   extentZ=0.5
   shapeName=dyn_money_02_top_bill_03_box
   center=<1.7677 0 0.0163>
   rotation=<rot 0.0013 -90 4.9988>
  }]
 dyn_money_02_top_bill_04 [{
   shapeType=Box
   extentX=8
   extentY=6
   extentZ=0.5
   shapeName=dyn_money_02_top_bill_04_box
   center=<2.9477 1.4719 -0.0908>
   rotation=<rot 0.0007 40.0001 0>
  }]
 dyn_top_handle [{shapeType=Capsule, length=3, radius=2, shapeName=dyn_top_handle_capsule, center=<2.1051 -0.3525 0>, rotation=<rot 0.0007 0 -90>}]
 dyn_zipper_01 [{shapeType=Box, extentX=2.3, extentY=0.6, extentZ=0.3, shapeName=dyn_zipper_01_box, center=<1.4516 0 0.0586>}]
 dyn_zipper_02 [{shapeType=Box, extentX=2.3, extentY=0.6, extentZ=0.3, shapeName=dyn_zipper_02_box, center=<1.4516 0 0.0586>}]
 money_02_root [
  {
   shapeType=Capsule
   length=16.25
   radius=4.75
   shapeName=money_02_root_capsule
   center=<-0.0682 -3.4833 4.5022>
   rotation=<rot -89.5629 92.482 -92.5015>
  }
  {
   shapeType=Capsule
   length=9.5
   radius=5.75
   shapeName=money_02_root_capsule
   center=<-3.7837 -7.3848 0.365>
   rotation=<rot -90 166.7373 -166.7376>
  }
 ]
 pack [{shapeType=Capsule, length=6.1422, radius=7.642, shapeName=pack_capsule, center=<0 -7.0151 -1.03>, rotation=<rot 90 -0.0099 179.9899>}]
 zipper_root [{shapeType=Box, extentX=4, extentY=3, extentZ=0.4, shapeName=zipper_root_box, center=<0 0.5352 -0.082>}]
}
constraints [
 {bone1Name=dyn_zipper_01, bone2Name=zipper_root, swing1LimitDegrees=60, swing2LimitDegrees=90, twistLimitDegrees=3}
 {bone1Name=dyn_zipper_02, bone2Name=zipper_root, swing1LimitDegrees=60, swing2LimitDegrees=90, twistLimitDegrees=3}
 {bone1Name=dyn_top_handle, bone2Name=pack, swing1LimitDegrees=25, swing2LimitDegrees=25, twistLimitDegrees=10}
 {bone1Name=dyn_money_02_left_bill_01, bone2Name=pack, swing1LimitDegrees=20, swing2LimitDegrees=40, twistLimitDegrees=5}
 {bone1Name=dyn_money_02_left_bill_02, bone2Name=pack, swing1LimitDegrees=20, swing2LimitDegrees=40, twistLimitDegrees=5}
 {bone1Name=dyn_money_02_right_bill_01, bone2Name=pack, swing1LimitDegrees=20, swing2LimitDegrees=40, twistLimitDegrees=5}
 {bone1Name=dyn_money_02_right_bill_02, bone2Name=pack, swing1LimitDegrees=20, swing2LimitDegrees=40, twistLimitDegrees=5}
 {bone1Name=dyn_money_02_top_bill_01, bone2Name=pack, swing1LimitDegrees=40, swing2LimitDegrees=30, twistLimitDegrees=5}
 {bone1Name=dyn_money_02_top_bill_03, bone2Name=pack, swing1LimitDegrees=20, swing2LimitDegrees=40, twistLimitDegrees=2}
 {bone1Name=dyn_money_02_top_bill_02, bone2Name=pack, swing1LimitDegrees=20, swing2LimitDegrees=30, twistLimitDegrees=2}
 {bone1Name=dyn_money_02_top_bill_04, bone2Name=pack, swing1LimitDegrees=20, swing2LimitDegrees=40, twistLimitDegrees=5}
]
