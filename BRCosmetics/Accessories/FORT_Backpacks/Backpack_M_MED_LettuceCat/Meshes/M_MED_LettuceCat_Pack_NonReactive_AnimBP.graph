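# Animation Blueprint for the non-reactive LettuceCat backpack, driving pack physics and stage-based material hiding.
# Three cached poses (input plus two additive bulge, zipper and flap poses) feed per-stage RigidBody nodes, but only stage 01 reaches Output Pose.
# A looping timer, faster for tagged front-end variants, always resolves to stage 01, leaving stage 02/03 setters unwired.
# Stage events set hide parameters on base and overlay materials and broadcast SetBackpackStage.
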AnimBlueprint /BRCosmetics/Accessories/FORT_Backpacks/Backpack_M_MED_LettuceCat/Meshes/M_MED_LettuceCat_Pack_NonReactive_AnimBP
target_skeleton=&/M_MED_LettuceCat_Pack_Skeleton
graphs [
 AnimGraph {nodes [
   n1 = RigidBody(ComponentPose=Local To Component(LocalPose=Use cached pose 'State_02'())) {
    bOverrideWorldGravity=true
    overridePhysicsAsset=&/M_MED_LettuceCat_Pack_Physics_stage_02
    simulationSpace=BaseBoneSpace
    alphaBoolBlend {blendOption=Linear}
    componentAppliedLinearAccClamp=<3000 3000 3000>
    componentLinearAccScale=<0.4 0.3 0.3>
    componentLinearVelScale=<0.4 0.3 0.3>
    overrideWorldGravity {z=-900}
   }
   n2 = RigidBody(ComponentPose=Local To Component(LocalPose=Use cached pose 'State_03'())) {
    bOverrideWorldGravity=true
    overridePhysicsAsset=&/M_MED_LettuceCat_Pack_Physics_stage_03
    simulationSpace=BaseBoneSpace
    alphaBoolBlend {blendOption=Linear}
    componentAppliedLinearAccClamp=<3000 3000 3000>
    componentLinearAccScale=<0.4 0.3 0.3>
    componentLinearVelScale=<0.4 0.3 0.3>
    overrideWorldGravity {z=-900}
   }
   Blend Poses by bool(
    BlendPose_0=Component To Local(ComponentPose=@n2)
    BlendPose_1=Blend Poses by bool(BlendPose_0=Component To Local(ComponentPose=@n1), bActiveValue=Get IsStage02() -> (IsStage02:bool)) {blendType=Linear}
    bActiveValue=Get IsStage03() -> (IsStage03:bool)
   ) {blendType=Linear}
   n3 = RigidBody(ComponentPose=Local To Component(LocalPose=Use cached pose 'State_01'())) {
    bOverrideWorldGravity=true
    overridePhysicsAsset=&/M_MED_LettuceCat_Pack_Physics_stage_01
    simulationSpace=BaseBoneSpace
    alphaBoolBlend {blendOption=Linear}
    componentAppliedLinearAccClamp=<3000 3000 3000>
    componentLinearAccScale=<0.4 0.3 0.3>
    componentLinearVelScale=<0.4 0.3 0.3>
    overrideWorldGravity {z=-900}
   }
   Output Pose(Result=Component To Local(ComponentPose=@n3)) {layerGroup=None}
   n4 = "Transform (Modify) Bone - Bone: top_bulge_upper"(ComponentPose=Local To Component(LocalPose=Use cached pose 'State_01'()), Alpha=1, Translation="0,-0.2,0", Rotation="0,0,0", Scale="1,1,1") {
    translationMode=BMM_Additive
    translationSpace=BCS_BoneSpace
    alphaBoolBlend {blendOption=Linear}
    boneToModify {boneName=top_bulge_upper}
    translation {y=-0.2}
   }
   n5 = "Transform (Modify) Bone - Bone: zipper_root"(ComponentPose=@n4, Alpha=1, Translation="5.5,0,0", Rotation="0,0,0", Scale="1,1,1") {
    translationMode=BMM_Additive
    translationSpace=BCS_BoneSpace
    alphaBoolBlend {blendOption=Linear}
    boneToModify {boneName=zipper_root}
    translation {x=5.5}
   }
   n6 = "Transform (Modify) Bone - Bone: top_bulge_lower"(ComponentPose=@n5, Alpha=1, Translation="0,-1,0", Rotation="0,0,0", Scale="1,1,1") {
    translationMode=BMM_Additive
    translationSpace=BCS_BoneSpace
    alphaBoolBlend {blendOption=Linear}
    boneToModify {boneName=top_bulge_lower}
    translation {y=-1}
   }
   n7 = "Transform (Modify) Bone - Bone: bottom_bulge_lower"(ComponentPose=@n6, Alpha=1, Translation="0,-1.5,0", Rotation="0,0,0", Scale="1,1,1") {
    translationMode=BMM_Additive
    translationSpace=BCS_BoneSpace
    alphaBoolBlend {blendOption=Linear}
    boneToModify {boneName=bottom_bulge_lower}
    translation {y=-1.5}
   }
   n8 = "Transform (Modify) Bone - Bone: bottom_bulge_upper"(ComponentPose=@n7, Alpha=1, Translation="0,-0.5,0", Rotation="0,0,0", Scale="1,1,1") {
    translationMode=BMM_Additive
    translationSpace=BCS_BoneSpace
    alphaBoolBlend {blendOption=Linear}
    boneToModify {boneName=bottom_bulge_upper}
    translation {y=-0.5}
   }
   Save cached pose 'State_02'(Pose=Component To Local(ComponentPose=@n8)) {cacheName=State_02}
   Save cached pose 'State_01'(Pose=Input Pose()) {cacheName=State_01}
   n9 = "Transform (Modify) Bone - Bone: zipper_root"(
    ComponentPose=Local To Component(LocalPose=Use cached pose 'State_01'())
    Alpha=1
    Translation="-9.448,7.411,-10.305"
    Rotation="85.000,180,-160"
    Scale="1,1,1"
   ) {
    rotationMode=BMM_Additive
    rotationSpace=BCS_BoneSpace
    translationMode=BMM_Additive
    translationSpace=BCS_BoneSpace
    alphaBoolBlend {blendOption=Linear}
    boneToModify {boneName=zipper_root}
    rotation=<rot 85 180 -160>
    translation=<-9.448 7.411 -10.305>
   }
   n10 = "Transform (Modify) Bone - Bone: top_flap"(ComponentPose=@n9, Alpha=1, Translation="0,0,0", Rotation="32,0,0", Scale="1,1,1") {
    rotationMode=BMM_Additive
    rotationSpace=BCS_BoneSpace
    alphaBoolBlend {blendOption=Linear}
    boneToModify {boneName=top_flap}
    rotation {pitch=32}
   }
   n11 = "Transform (Modify) Bone - Bone: top_bulge_upper"(ComponentPose=@n10, Alpha=1, Translation="0,-1.6,-1", Rotation="0,0,0", Scale="1,1,1") {
    rotationMode=BMM_Additive
    translationMode=BMM_Additive
    translationSpace=BCS_BoneSpace
    alphaBoolBlend {blendOption=Linear}
    boneToModify {boneName=top_bulge_upper}
    translation {y=-1.6, z=-1}
   }
   n12 = "Transform (Modify) Bone - Bone: top_bulge_lower"(ComponentPose=@n11, Alpha=1, Translation="0,-2,-1", Rotation="0,0,0", Scale="1,1,1") {
    rotationMode=BMM_Additive
    translationMode=BMM_Additive
    translationSpace=BCS_BoneSpace
    alphaBoolBlend {blendOption=Linear}
    boneToModify {boneName=top_bulge_lower}
    translation {y=-2, z=-1}
   }
   n13 = "Transform (Modify) Bone - Bone: bottom_bulge_upper"(ComponentPose=@n12, Alpha=1, Translation="0,-2,0", Rotation="0,0,0", Scale="1,1,1") {
    translationMode=BMM_Additive
    translationSpace=BCS_BoneSpace
    alphaBoolBlend {blendOption=Linear}
    boneToModify {boneName=bottom_bulge_upper}
    translation {y=-2}
   }
   n14 = "Transform (Modify) Bone - Bone: bottom_bulge_lower"(ComponentPose=@n13, Alpha=1, Translation="0,-3,0", Rotation="0,0,0", Scale="1,1,1") {
    translationMode=BMM_Additive
    translationSpace=BCS_BoneSpace
    alphaBoolBlend {blendOption=Linear}
    boneToModify {boneName=bottom_bulge_lower}
    translation {y=-3}
   }
   Save cached pose 'State_03'(Pose=Component To Local(ComponentPose=@n14)) {cacheName=State_03}
  ]}
 CheckHideElements {nodes [
   Inputs()
   n1 = GetPartSkeletalMeshComponent()
   > n2 = Get Dynamic Material Instance(SkeletalMeshComponent=@n1)
   > n3 = Get Scalar Parameter Value!(execute, self:Material Instance Dynamic Object Reference=@n2, ParameterName:Name=Hide Element 01) -> (then, ReturnValue:float) {is=title K2Node_CallFunction}
   > Set IsHideElement01(IsHideElement01:bool="Equal (Float)"(A=@n3)) -> (Output_Get:bool)
   > n4 = Get Dynamic Material Instance(SkeletalMeshComponent=@n1)
   > n5 = Get Scalar Parameter Value(self=@n4, ParameterName=Hide Element 02)
   > Set IsHideElement02(IsHideElement02:bool="Equal (Float)"(A=@n5)) -> (Output_Get:bool)
   > Outputs!(then:Exec)
  ]}
 EventGraph {nodes [
   Event SetStage03() {is=title K2Node_CustomEvent}
   n1 = GetPartSkeletalMeshComponent()
   > n2 = Get Dynamic Material Instance(SkeletalMeshComponent=@n1)
   > SetScalarParameterValue(self=@n2, ParameterName=Hide Element 02, Value=0)
   > SetScalarParameterValue(self=@n2, ParameterName=Hide Element 01, Value=1)
   > SetScalarParameterValueOnOverlayMaterial!(execute, MeshComponent:Mesh Component Object Reference=@n1, ParameterName:Name=Hide Element 02, ParameterValue:float=0) -> (then)
   > SetScalarParameterValueOnOverlayMaterial(MeshComponent=@n1, ParameterName=Hide Element 01, ParameterValue=1)
   > Call Set Backpack Stage!(execute:Exec, self:M MED Lettuce Cat Pack Non Reactive Anim BP Object Reference, Stage:int=2) -> (then:Exec) {is=title K2Node_CallDelegate}
   Event SetStage01() {is=title K2Node_CustomEvent}
   > n3 = Get Dynamic Material Instance(SkeletalMeshComponent=@n1)
   > SetScalarParameterValue(self=@n3, ParameterName=Hide Element 02, Value=1)
   > SetScalarParameterValue(self=@n3, ParameterName=Hide Element 01, Value=1)
   > SetScalarParameterValueOnOverlayMaterial(MeshComponent=@n1, ParameterName=Hide Element 02, ParameterValue=1)
   > SetScalarParameterValueOnOverlayMaterial(MeshComponent=@n1, ParameterName=Hide Element 01, ParameterValue=1)
   > Call Set Backpack Stage()
   n4 = UpdateMoneyAmount() {is=title K2Node_CustomEvent}
   n5 = AND Boolean(
    A=Get bIsInFrontEnd()
    B=BackpackItemDefinitionHasAnyMetaTag!(Tags:Array of Gameplay Tag Structures=Make Array("[0]"=(TagName="Cosmetics.Variant.Property.Stage1")), bRequiresEquipped:bool=false) -> (ReturnValue:bool)
   )
   > n6 = Branch(Condition=@n5)
   > n7 = MultiGate!(execute:Exec, Reset:Exec, IsRandom:bool, Loop:bool=@n5, StartIndex:int=1) -> (Out 0:Exec, Out 1:Exec, Out 2:Exec) {is=title K2Node_MultiGate}
   Set IsStage02(execute=[@n11.then, @n8.Out 0, @n8.Out 1, @n7.Out 0, @n7.Out 1, @n7.Out 2], IsStage02:bool=false) -> (Output_Get:bool)
   > Set IsStage03(IsStage03:bool=false) -> (Output_Get:bool)
   > Event SetStage01() {is=title K2Node_CallFunction}
   n8 = MultiGate!(execute=@n6.else, Reset, IsRandom, Loop, StartIndex=-1) -> (Out 0, Out 1)
   Event SetStage02() {is=title K2Node_CustomEvent}
   > n9 = Get Dynamic Material Instance(SkeletalMeshComponent=@n1)
   > SetScalarParameterValue(self=@n9, ParameterName=Hide Element 02, Value=1)
   > SetScalarParameterValue(self=@n9, ParameterName=Hide Element 01, Value=0)
   > SetScalarParameterValueOnOverlayMaterial(MeshComponent=@n1, ParameterName=Hide Element 02, ParameterValue=1)
   > SetScalarParameterValueOnOverlayMaterial(MeshComponent=@n1, ParameterName=Hide Element 01, ParameterValue=0)
   > Call Set Backpack Stage(Stage=1)
   n10 = Cast To PlayerPawn_Athena(Object=TryGetPawnOwner())
   > Set As Player Pawn Athena(As Player Pawn Athena:Fort Player Pawn Athena Object Reference=@n10) -> (Output_Get:Fort Player Pawn Athena Object Reference)
   Event OnCharacterCustomizationLoaded()
   > n11 = Set Timer by Event(Delegate=@n4, Time=SelectFloat(A=5, B=180, bPickA=@n5), bLooping=true)
   Set IsStage02(IsStage02=true)
   > Set IsStage03(IsStage03=false)
   > Event SetStage02() {is=title K2Node_CallFunction}
   Set IsStage02(IsStage02=false)
   > Set IsStage03(IsStage03=true)
   > Event SetStage03() {is=title K2Node_CallFunction}
   Is Valid(InputObject=Get As Player Pawn Athena() -> (As Player Pawn Athena:Fort Player Pawn Athena Object Reference))
  ]}
 SetBackpackStage {nodes [SetBackpackStage!() -> (then, Stage:int) {is=title K2Node_FunctionEntry}]}
]
variables [
 As Player Pawn Athena {type=FortPlayerPawnAthena}
 Frontend Money {type=number}
 IsHideElement01 {type=boolean}
 IsHideElement02 {type=boolean}
 IsStage02 {type=boolean}
 IsStage03 {type=boolean}
 SetBackpackStage {default=()}
 Start Time {type=number}
 Time in Round {type=number}
]
